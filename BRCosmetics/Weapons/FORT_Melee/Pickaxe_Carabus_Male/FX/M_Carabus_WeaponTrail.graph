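# Translucent unlit weapon trail material for the Carabus pickaxe, usable on Niagara ribbons and mesh particles.
# Two noise samples pan at different speeds and are combined, inverted and shaped by sphere masks and a U-gradient erosion into depth-faded opacity scaled by particle alpha.
# Dynamic parameters drive tiling, scroll speed and noise power.
# A two-color gradient along U passes through MF_FX_EmissiveAdjustments for the glow color.
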Material /BRCosmetics/Weapons/FORT_Melee/Pickaxe_Carabus_Male/FX/M_Carabus_WeaponTrail
graph {
 properties {
  bUsedWithNiagaraMeshParticles=true
  bUsedWithNiagaraRibbons=true
  blendMode=BLEND_Translucent
  pixelDepthOffsetMode=PDOM_Legacy
  shadingModel=MSM_Unlit
  usedShadingModels=Unlit
 }
 nodes [
  n1 = DynamicParameter() {paramNames [TilingOffset,TilingMult,ScrollSpeed,TexturePower]}
  n2 = Constant() {r=1}
  n3 = Multiply(
   A=Add(A=TextureCoordinate() {uTiling=2}, B=@n1.TilingOffset)
   B=ComponentMask(None=MaterialFunctionCall(X=@n1.TilingMult, Y=@n2) {materialFunction=&MakeFloat2}.Result) {g=true, r=true}
  )
  n4 = Multiply(A=ScalarParameter() {defaultValue=0.4, parameterName=UV1SpeedModifier}, B=@n1.ScrollSpeed)
  n5 = Constant()
  n6 = TextureSample(UVs=Panner(
    Coordinate=@n3
    Speed=MaterialFunctionCall(X=@n4, Y=@n5) {materialFunction=&MakeFloat2}.Result
   ) {speedX=0.4}) {texture=&T_NoisePack_01}
  n7 = Multiply(A=ScalarParameter() {defaultValue=0.8, parameterName=UV2SpeedModifier}, B=@n1.ScrollSpeed)
  n8 = TextureSample(UVs=Panner(
    Coordinate=@n3
    Speed=MaterialFunctionCall(X=@n7, Y=@n5) {materialFunction=&MakeFloat2}.Result
   ) {speedX=0.8}) {texture=&T_NoisePack_01}
  n9 = Multiply(
   A=OneMinus(None=Max(A=@n6.R, B=@n8.G))
   B=SphereMask(A=TextureCoordinate() {uTiling=0}, B=Constant2Vector() {g=0.5}) {attenuationRadius=0.5, hardnessPercent=25}
  )
  n10 = DynamicParameter() {paramNames [TilingOffset,TilingMult,ScrollSpeed,TexturePower]}
  n11 = ComponentMask(None=TextureCoordinate()) {r=true}
  n12 = Divide(A=Subtract(A=Power(Base=@n9, Exp=@n10.TexturePower) {constExponent=1}, B=Power(Base=@n11)), B=Constant() {r=0.5403})
  n13 = SphereMask(A=TextureCoordinate() {vTiling=0}, B=Constant2Vector() {r=1.05}) {attenuationRadius=1, hardnessPercent=70.4762}
  n14 = Multiply(
   A=Multiply(A=Saturate(None=@n12), B=Power(Base=@n13) {constExponent=4})
   B=ScalarParameter() {defaultValue=0.5, parameterName=Opacity Multiply}
  )
  DepthFade(Opacity=Multiply(A=ParticleColor().A, B=Saturate(None=@n14)), FadeDistance=ScalarParameter() {defaultValue=8, parameterName=DepthFade}) {fadeDistanceDefault=8}
  n15 = Add(A=ComponentMask(None=TextureCoordinate()) {r=true}, B=Constant() {r=-0.1})
  n16 = Constant() {r=2}
  MaterialFunctionCall(Emissive=LinearInterpolate(
    A=VectorParameter() {parameterName=Color1}.RGB
    B=VectorParameter() {parameterName=Color2, defaultValue=<rgba 0 0 0 1>}.RGB
    Alpha=Multiply(A=@n15, B=@n16)
   )) {materialFunction=&MF_FX_EmissiveAdjustments}
  Multiply(A=Multiply(A=@n6.RGB, B=@n8.RGB), B=DynamicParameter() {paramNames [TilingOffset,TilingMult,ScrollSpeed,TexturePower]}.TilingMult)
 ]
}
